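; Weapon Blueprint for the DragonCart SMG that switches its visual effects between hip fire and aim down sights.
; BeginPlay binds the ADS, ammo-consume and beam-fired delegates, caching the ADS state in a bool.
; Ammo consumption spawns either the ADS or master Niagara muzzle flash.
; Beam and impact handlers pass IsADS to the particles, adding beam endpoints and direction or surface type, and impacts also set mode-specific decal colors.
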(event EventBeginPlay
  ; BEGIN PLAY
  (|CallParentFunction)
  (Default|BindEventtoOnAimDownSightsChanged (EventDispatchers|CreateEvent :!event Custom|OnADS))
  (EventDispatchers|BindEventtoOnConsumeAmmo (EventDispatchers|CreateEvent :!event Custom|OnConsumeAmmo_Event))
  (Variables|FortWeapon|SetOnBeamFiredDelegate (EventDispatchers|CreateEvent :!event Custom|OnBeamFired)))

(event Custom|OnADS
  (params (AimDownSights :type Boolean))
  ; ADS BOOL
  (Variables|Default|SetAim%20Down%20Sights AimDownSights))

(event Custom|OnBeamFired
  (params (Weapon :type Fort%20Weapon%20Object%20Reference) (bUsePersistentBeam :type Boolean) (PSC :type FXSystem%20Component%20Object%20Reference) (BeamOrigin :type Vector%20%28by%20ref%29) (HitLocation :type Vector%20%28by%20ref%29))
  ; TRAIL OVERRIDE FOR ADS VARIANT
  (bind _asniagara_particle_system_component (Utilities|Casting|CastToNiagaraComponent PSC))
  (Niagara|SetNiagaraVariable(Bool) _asniagara_particle_system_component "IsADS" (Variables|Default|GetAim%20Down%20Sights))
  (Niagara|SetNiagaraVariable(Vector3) _asniagara_particle_system_component "BeamStart_0_0" BeamOrigin)
  (Niagara|SetNiagaraVariable(Vector3) _asniagara_particle_system_component "BeamEnd_0" HitLocation)
  (Niagara|SetNiagaraVariable(Vector3) _asniagara_particle_system_component "Direction" (Math|Vector|GetRightVector (Transformation|GetWorldRotation (Variables|Visual|GetWeaponMesh)))))

(event FortWeapon|EventOnPlayImpactFX
  (params (HitResult :type Hit%20Result%20Structure%20%28by%20ref%29) (ImpactPhysicalSurface :type EPhysicalSurface%20Enum :default "SurfaceType_Default") (SpawnedPSC :type FXSystem%20Component%20Object%20Reference))
  ; IMPACT OVERRIDE FOR ADS VARIANT
  (|CallParentFunction HitResult ImpactPhysicalSurface SpawnedPSC)
  (Effects|Components|ParticleSystem|SetBoolParameter SpawnedPSC "IsADS" (Variables|Default|GetAim%20Down%20Sights))
  (Effects|Components|ParticleSystem|SetIntParameter SpawnedPSC "PhysicalSurface" (Math|Conversions|ToInteger(Byte) ImpactPhysicalSurface))
  (if (Variables|Default|GetAim%20Down%20Sights)
    (Decals|SetDecalColorStart "(R=10.000000,G=0.658333,B=0.500000,A=1.000000)")
    (Decals|SetDecalColorEnd "(R=0.500000,G=0.005000,B=0.016086,A=1.000000)")
    (else
      (Decals|SetDecalColorStart "(R=8.000000,G=1.000000,B=0.250000,A=1.000000)")
      (Decals|SetDecalColorEnd "(R=1.000000,G=0.092500,B=0.010000,A=1.000000)"))))

(event Custom|OnConsumeAmmo_Event
  (params (Quantity :type Integer))
  ; MF OVERRIDE FOR ADS VARIANT
  (if (Variables|Default|GetAim%20Down%20Sights)
    (Niagara|SpawnSystemAttached "/DragonCartWeaponGameplay/Gameplay/DragonCart_SMG/SMG_Standard_DragonCart/MuzzleVFX/NS_DualModeSMG_MuzzleFlash_ADS.NS_DualModeSMG_MuzzleFlash_ADS" (Variables|Visual|GetWeaponMesh) "Muzzle" "0, 0, 0" "0, 0, 0" "SnapToTargetIncludingScale" true)
    (else
      (Niagara|SpawnSystemAttached "/Game/Effects/Niagara/Systems/Weapon_Masters/NS_MuzzleFlash_Master.NS_MuzzleFlash_Master" (Variables|Visual|GetWeaponMesh) "Muzzle" "0, 0, 0" "0, 0, 0" "SnapToTargetIncludingScale" true))))

(fn ConstructionScript
  (|CallParentFunction))
