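# Niagara tracer effect for the DragonCart SMG, drawing a beam from the weapon toward a user-supplied end point.
# One emitter bursts two spawn groups into two ribbon renderers and a mesh bullet, tagged by group, with travel along the beam driven by frame-limited age.
# Beam length scales lifetime and travel timing, while LOD distance enlarges the bullet.
# Curves shape width and fade, and an ADS flag switches the tint to intense red.
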NiagaraSystem /DragonCartWeaponGameplay/Gameplay/DragonCart_SMG/SMG_Standard_DragonCart/VFX/NS_DragonCart_SMG_Tracer
dependencies {
 usedDataInterfaces [&NiagaraDataInterfaceCurve]
 usedDynamicInputs [&NormalizeDistanceRange]
 usedModules [
  &self:SystemSpawnScript.SystemScriptSource.SystemScriptGraph.NiagaraNodeAssignment_2.SetVariables_#1
  &self:SystemSpawnScript.SystemScriptSource.SystemScriptGraph.NiagaraNodeAssignment_0.SetVariables_#2
  &self:SystemSpawnScript.SystemScriptSource.SystemScriptGraph.NiagaraNodeAssignment_1.SetVariables_#3
  &SystemState
 ]
}
emitters {DynamicBeam {
  emitterSpawnScript [/SetVariables(Emitter.BeamEnd:Vector3f=$User.BeamEnd_0, Emitter.DebugTimeMultiplier=1) {isSetParameters=true}]
  emitterUpdateScript [
   &EmitterState()
   /Niagara/Modules/Beams/BeamEmitterSetup(
    Absolute Beam End=true
    Beam End:Vector3f=$Emitter.BeamEnd
    Beam End Tangent:Vector3f=&Multiply_VectorByFloat(Vector=<0 0 0>, Float=1)
    Beam Start Tangent:Vector3f=&Multiply_VectorByFloat(
     Vector=&NormalizeVector(Vector To Normalize=$System.Direction, Safe Normalize=true, Safe Fallback Threshold=0, Safe Fallback Vector=<0 0 0.001>)
     Float=&MakeCustomFloatFromBool(Bool=false, True Float=&RandomRangeFloat(Minimum=0.1, Maximum=0.4), False Float=0)
    )
    Use Beam Tangents=true
   )
   &SpawnBurst_Instantaneous(Spawn Count=9, Spawn Time=0.001, Use Spawn Probability=false, UseLoopCountLimit=false)
   &SpawnBurst_Instantaneous(Spawn Group=1, Use Spawn Probability=false, UseLoopCountLimit=false) {module=SpawnBurst_Instantaneous001}
  ]
  particleSpawnScript [
   /SetVariables(Particles.VisibilityTag=&SpawnGroup()) {isSetParameters=true}
   /SetVariables(
    Particles.BeamTile:NiagaraFloat=&Divide_Float(
     A=/Niagara/DynamicInputs/TypeConversions/VectorLengthSafe(Vector:Vector3f=&Subtract_Vector(A=Linked() {linkedVariable {name=Engine.Owner.Position, type=NiagaraPosition}}, B=$Emitter.BeamEnd))
     B=500
    )
    Particles.FireDirection:Vector3f=/Niagara/DynamicInputs/Helpers/SafeNormalizeVector(
     Vector To Normalize:Vector3f=&Subtract_Vector(A=$Emitter.BeamEnd, B=Linked() {linkedVariable {name=Engine.Owner.Position, type=NiagaraPosition}})
     Fallback Vector:Vector3f=$Engine.Owner.SystemXAxis
    )
   ) {isSetParameters=true}
   /SetVariables(
    Particles.BeamDistanceScalar:NiagaraFloat=&Lerp_Float(
     A=0.35
     B=1
     Alpha=&FloatFromCurve(
      FloatCurve=DataInterface() {propertyValues {Curve {keys [
          {arriveTangent=0.0002, interpMode=RCIM_Cubic, leaveTangent=0.0002, tangentMode=RCTM_User, time=0, value=0}
          {arriveTangent=0.0001, interpMode=RCIM_Cubic, leaveTangent=0.0001, tangentMode=RCTM_User, time=2500, value=0.35}
          <at 15000 0.8>
          {interpMode=RCIM_Cubic, tangentMode=RCTM_User, time=35000, value=1}
         ]}}}
      CurveIndex=&VectorLengthSafe(Vector:Vector3f=&Subtract_Vector(A=Linked() {linkedVariable {name=Engine.Owner.Position, type=NiagaraPosition}}, B=$Emitter.BeamEnd))
     )
    )
    Particles.LODDistanceRemap:NiagaraFloat=&FloatFromCurve(FloatCurve=DataInterface() {propertyValues {Curve {keys [<at 500 0>,<at 2500 1>]}}}, CurveIndex=$Engine.Owner.LODDistance)
   ) {isSetParameters=true}
   &InitializeParticle(
    Lifetime=&Multiply_Float(
     A=&ClampFloat(Float=$Particles.BeamDistanceScalar, Min=0.5)
     B=&MaskFloatBySpawnGroup(
      Masked Float=&RandomRangeFloat(Minimum=1, Maximum=1.25)
      Passthrough Float=&RandomRangeFloat(Minimum=0.7, Maximum=0.8)
     )
    )
    Mesh Scale Mode=::NewEnumerator3
    Mesh Uniform Scale=0.9
    Position Mode=::NewEnumerator0
    Ribbon Width Mode=::NewEnumerator1
    Write Material Random=true
   )
   /Niagara/Modules/Beams/SpawnBeam()
   /SetVariables(Particles.BulletDynParams:Vector4f=<1 1 1 1>) {isSetParameters=true}
   /SetVariables() {isSetParameters=true}
   /SetVariables(Particles.RibbonU0Override=&OneMinusFloat(Float=$Particles.RibbonLinkOrder)) {isSetParameters=true}
  ]
  particleUpdateScript [
   &ParticleState()
   /Niagara/Modules/Update/Utility/FrameCounter()
   /SetVariables(Particles.FrameLimitedAge:NiagaraFloat=&self:ScratchDynamicInput_07(
     MaxDuration:NiagaraFloat=0.5
     MinDuration:NiagaraFloat=0.1
     CurrentFrame:NiagaraFloat=&MakeFloatFromInt(Int=$Particles.FrameCounter.Counter)
     MaxFrames:NiagaraFloat=&MakeFloatFromInt(Int=8)
     DurationMultiplier:NiagaraFloat=$Particles.BeamDistanceScalar
     DebugTimeMultiplier:NiagaraFloat=1
    )) {isSetParameters=true}
   /SetVariables(Particles.Travel:NiagaraFloat=&self:RemapRange_08(
     Value:NiagaraFloat=$Particles.FrameLimitedAge
     Input Min:NiagaraFloat=0
     Input Max:NiagaraFloat=0.5
     Output Min:NiagaraFloat=0
     Output Max:NiagaraFloat=1
    )) {isSetParameters=true}
   /SetVariables(Particles.RibbonLinkOrder:NiagaraFloat=&MaskFloatBySpawnGroup(
     Masked Float=$Particles.RibbonLinkOrder
     Passthrough Float=&FloatFromCurve(
      FloatCurve=DataInterface() {propertyValues {Curve {keys [{tangentMode=RCTM_User, time=0, value=0}, {tangentMode=RCTM_User, time=1, value=1}]}}}
      CurveIndex=$Particles.Travel
     )
    )) {isSetParameters=true}
   /SetVariables(Particles.BulletVisbility:NiagaraInt32=&MakeCustomIntFromBool(
     Bool=&Boolean_AndOperation(
      Bool1=&SetBoolByFloatComparison(A=$Particles.FrameLimitedAge, B=0.55, Comparison Type=::NewEnumerator6)
      Bool2=&SetBoolByFloatComparison(A=$Particles.BeamTile, B=1)
     )
     True INT=0
     False INT=1
    )) {isSetParameters=true}
   /SetVariables(Particles.BulletSize:Vector3f=&Multiply_VectorByFloat(
     Vector=&Lerp_Vector(A=<1.5 1 1>, B=<3 2 2>, Alpha=$Particles.LODDistanceRemap)
     Float=&FloatFromCurve(
      FloatCurve=DataInterface() {propertyValues {Curve {keys [{arriveTangent=32.8701, time=0, value=0}, <at 0.05 1>]}}}
      CurveIndex=$Particles.FrameLimitedAge
     )
    )) {isSetParameters=true}
   /Niagara/Modules/Beams/UpdateBeam()
   /Niagara/Modules/Beams/ScaleBeamWidth(Beam Width Scale:NiagaraFloat=&FloatFromCurve(
     FloatCurve=DataInterface() {propertyValues {Curve {keys [{arriveTangent=3.6143, interpMode=RCIM_Cubic, leaveTangent=3.6143, tangentMode=RCTM_User, time=0, value=1}, <at 1 3 RCIM_Cubic>]}}}
     Scale Curve=0.75
    ))
   /SetVariables(
    Particles.ColorScifi:LinearColor=&MultiplyLinearColor(
     Linear Color=<rgba 1 1 1 1>
     Scale Factor=&MakeVector4FromVectorAndFloat(
      "Vector (RGB)"=<1 1 1>
      "Float (Alpha)"=&FloatFromCurve(
       FloatCurve=DataInterface() {propertyValues {Curve {
          defaultValue=1
          keys [
           {interpMode=RCIM_Cubic, tangentMode=RCTM_User, time=0, value=1}
           {arriveTangent=-0.0383, interpMode=RCIM_Cubic, leaveTangent=-0.0383, tangentMode=RCTM_User, time=1, value=0}
          ]
         }}}
       CurveIndex=&NormalizeFloat(
        Value=&Multiply_Float(A=&RandomRangeFloat(Minimum=1.5, Maximum=3), B=$Particles.NormalizedAge)
        Min=0
        Max=1
       )
      )
     )
    )
    Particles.DMP_Scifi:Vector4f=&MakeVector4(
     X=0
     Y=&RandomRangeFloat(Minimum=0, Maximum=1)
     Z=&RandomRangeFloat(Minimum=0, Maximum=1)
     W=1
    )
    Particles.RibbonWidthScifi:NiagaraFloat=&Multiply_Float(
     A=$Particles.RibbonWidth
     B=&FloatFromCurve(
      FloatCurve=DataInterface() {propertyValues {Curve {
         defaultValue=1
         keys [
          {arriveTangent=3.935, interpMode=RCIM_Cubic, leaveTangent=3.935, tangentMode=RCTM_User, time=0, value=0}
          {arriveTangent=0.336, interpMode=RCIM_Cubic, leaveTangent=0.336, tangentMode=RCTM_User, time=0.0636, value=0.5198}
          {arriveTangent=-0.0403, interpMode=RCIM_Cubic, leaveTangent=-0.0403, tangentMode=RCTM_User, time=0.9615, value=0.5814}
          {arriveTangent=-6.601, interpMode=RCIM_Cubic, leaveTangent=-6.601, tangentMode=RCTM_User, time=1.001, value=0.0145}
         ]
        }}}
      CurveIndex=$Particles.RibbonLinkOrder
      Scale Curve=&RandomRangeFloat(Minimum=2, Maximum=2.5)
     )
    )
   ) {isSetParameters=true}
   &ScaleColor(
    Scale Alpha=&FloatFromCurve(FloatCurve=DataInterface() {propertyValues {Curve {keys [
         {arriveTangent=-5.6727, interpMode=RCIM_Cubic, leaveTangent=-5.6727, tangentMode=RCTM_User, time=0.4, value=1}
         {interpMode=RCIM_Cubic, tangentMode=RCTM_User, time=1, value=0}
        ]}}})
    Scale RGB=&VectorFromFloat(Value=&FloatFromCurve(FloatCurve=DataInterface() {propertyValues {Curve {keys [{tangentMode=RCTM_User, time=0, value=1}, {tangentMode=RCTM_User, time=1, value=0}]}}}))
    ScaleA=true
    ScaleRGB=true
   )
   &SolveForcesAndVelocity(Clamp Velocity=false, Limit Acceleration=false)
   &DynamicMaterialParameters(
    Index 0 Param 1=2
    Index 0 Param 2=&FloatFromCurve(FloatCurve=DataInterface() {propertyValues {Curve {keys [{interpMode=RCIM_Cubic, tangentMode=RCTM_User, time=0, value=0}, {interpMode=RCIM_Cubic, tangentMode=RCTM_User, time=0.5, value=2}]}}})
    Index 0 Param 4=&FloatFromCurve(FloatCurve=DataInterface() {propertyValues {Curve {keys [{interpMode=RCIM_Cubic, tangentMode=RCTM_User, time=0, value=1}, {interpMode=RCIM_Cubic, tangentMode=RCTM_User, time=1, value=0.33}]}}})
    Index 1 Param 1=&FloatFromCurve(
     FloatCurve=DataInterface() {propertyValues {Curve {keys [
         {arriveTangent=-3.8765, interpMode=RCIM_Cubic, leaveTangent=-3.8765, tangentMode=RCTM_User, time=0, value=1}
         {interpMode=RCIM_Cubic, tangentMode=RCTM_User, time=0.66, value=0}
        ]}}}
     Scale Curve=0.5
    )
    Module.Param0WriteEnabled=true
    Module.Param0WriteEnabled1=true
    Module.Param1WriteEnabled=false
    Module.Param1WriteEnabled1=false
    Module.Param2WriteEnabled=false
    Module.Param2WriteEnabled1=false
    Module.Param3WriteEnabled=true
    Module.Param3WriteEnabled1=false
    Write Parameter Index 1=true
   )
   &UpdateMeshOrientation(
    Facing Interpolation=false
    "Facing Origin ":NiagaraPosition=$Particles.Previous.Position
    Facing Position:NiagaraPosition=$Particles.Position
    Orientation Method=::NewEnumerator2
    Override Quat=false
    TransformFacingOrigin=false
    TransformFacingPosition=false
    TransformUp=false
   )
   &ScaleColor(
    Color Scale:LinearColor=&MakeCustomLinearColorFromBool(Bool=$User.IsADS, True Color=<rgba 5 0 0 1>, False Color=<rgba 1 0.3537 0.3114 1>)
    Scale Mode=3
   ) {module=ScaleColor001}
  ]
  rendererClasses [&NiagaraRibbonRendererProperties, &NiagaraMeshRendererProperties]
  renderers [
   {rendererClass=&NiagaraRibbonRendererProperties}
   {rendererClass=&NiagaraRibbonRendererProperties, rendererIndex=1}
   {rendererClass=&NiagaraMeshRendererProperties, rendererIndex=2}
  ]
 }}
summary {
 systemName=NS_DragonCart_SMG_Tracer
 emitters [{emitterName=DynamicBeam, rendererClasses [&NiagaraRibbonRendererProperties, &NiagaraMeshRendererProperties]}]
 userVariables [
  User.BeamEnd_0 {type=Vector3f, description="", defaultValue {struct=Vector3f, value=<0 0 2100>}}
  User.Direction {type=Vector3f, description="", defaultValue {struct=Vector3f, value=<1 0 0>}}
  User.IsADS {type=NiagaraBool}
  User.MaxDistance {type=NiagaraFloat, defaultValue=27000, description=""}
 ]
}
