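# Niagara data channel handler system that draws bullet impact effects for the DragonCart SMG, spawning from shared Fort impact channel data.
# Emitters add a mesh and sprite flash, glowing fading decals, smoke, sparks, streaks and physical-surface-specific mesh debris.
# Particles take position, surface normal, ammo type and distance from the channel to drive direction, scale and LOD.
# A user hue control tints flash, sparks and streaks, which also fade near the camera.
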NiagaraSystem /DragonCartWeaponGameplay/Gameplay/DragonCart_SMG/SMG_Standard_DragonCart/VFX/NS_NDC_DragonCart_SMG_Handler_Impacts
dependencies {
 usedDataInterfaces [
  &NiagaraDataInterfaceArrayFloat
  &NiagaraDataInterfaceCurve
  &NiagaraDataInterfaceSpriteRendererInfo
  &NiagaraDataInterfaceCamera
  &NiagaraDataInterfaceVector2DCurve
  &NiagaraDataInterfaceArrayInt32
  &NiagaraDataInterfaceArrayColor
  /Script/Niagara.NiagaraDataInterfaceDataChannelRead
 ]
 usedDynamicInputs []
 usedModules [
  &SystemState
  &self:SystemSpawnScript.SystemScriptSource.SystemScriptGraph.NiagaraNodeAssignment_1.SetVariables_#1
  /Niagara/Modules/System/CompleteIfUnused
 ]
}
emitters {
 Base_Flash {
  emitterUpdateScript [
   &EmitterState()
   /Game/Effects/Niagara/DataChannels/Impacts/Spawn_NDCFortImpacts()
  ]
  particleSpawnScript [
   &self:Initialize_Particle_from_NDC_PiewPistol_Impacts(
    AmmoType Scale=1
    Distance Scale=1
    LOD Input Distance Max=12000
    LOD Input Distance Min=750
    LOD Output Scale Max=1
    LOD Output Scale Min=0
    Scale Array=DataInterface() {propertyValues {GpuSyncMode=SyncCpuToGpu, MaxElements=4, FloatData [0.8,1,1.25,0.6,1]}}
   ) {module=Initialize_Particle_from_NDC_PiewPistol_Impacts}
   &InitializeParticle(
    Color=&self:Material_Hue_Shift(Color:LinearColor=<rgba 1 0.3636 0.01 1>, Hue Shift:NiagaraFloat=$User.ImpactColorHueControl)
    Lifetime Max=0.15
    Lifetime Min=0.11
    Lifetime Mode=::NewEnumerator1
    Mesh Scale=&Multiply_VectorByFloat(
     Vector=&RandomRangeVector(Minimum=<0.7 0.7 0.7>, Maximum=<0.9 0.9 0.9>)
     Float=$Particles.AmmoTypeScale
    )
    Mesh Scale Mode=::NewEnumerator4
    Position=$Particles.Position
    Position Mode=::NewEnumerator1
    Sprite Size Mode=::NewEnumerator3
    Uniform Sprite Size=&Multiply_Float(A=&RandomRangeFloat(Minimum=50, Maximum=60), B=$Particles.AmmoTypeScale)
    UsePositionOffset=false
    Write Material Random=true
   )
   &SubUVAnimation(
    Evaluation Type=::NewEnumerator0
    Output Alternate Renderer Binding=true
    "Sprite Renderer "=DataInterface() {propertyValues {SpriteRenderer=&self:Base_Flash_0.NiagaraSpriteRendererProperties_1}}
    SubUV Animation Mode:ENiagara_SubUVLookupModeV2=::NewEnumerator2
    UseEndFrame=false
    UseStartFrame=false
   )
   &CameraOffset(Camera Offset Amount=20)
   &InitialMeshOrientation(
    Mesh Orientation Mode=::NewEnumerator2
    Orientation Vector=$Particles.ImpactSurfaceNormal
    Rotation=&RandomRangeVector(Minimum=<0 -0.05 -0.05>, Maximum=<1 0.05 0.05>)
    Use Rotation Vector=true
   )
   /SetVariables() {isSetParameters=true}
  ]
  particleUpdateScript [
   &ParticleState()
   &ScaleMeshSize(Scale Factor=&VectorFromFloat(Value=&Multiply_Float(
      A=&FloatFromCurve(
       FloatCurve=DataInterface() {propertyValues {Curve {keys [
           {interpMode=RCIM_Cubic, tangentMode=RCTM_Break, time=0, value=0.5}
           {interpMode=RCIM_Cubic, tangentMode=RCTM_Break, time=0.7, value=1}
           <at 1 1>
          ]}}}
       CurveIndex=$Particles.DistanceScale
      )
      B=&FloatFromCurve(FloatCurve=DataInterface() {propertyValues {Curve {keys [{arriveTangent=4.2374, interpMode=RCIM_Cubic, leaveTangent=4.2374, tangentMode=RCTM_User, time=0, value=0}, <at 1 1.5>]}}})
     )))
   &ScaleSpriteSize(
    Scale Sprite Size Mode=::NewEnumerator0
    Uniform Scale Factor=&Multiply_Float(
     A=&FloatFromCurve(
      FloatCurve=DataInterface() {propertyValues {Curve {keys [
          {interpMode=RCIM_Cubic, tangentMode=RCTM_Break, time=0, value=0.6}
          {interpMode=RCIM_Cubic, tangentMode=RCTM_Break, time=0.7, value=1}
          <at 1 1>
         ]}}}
      CurveIndex=$Particles.DistanceScale
     )
     B=&FloatFromCurve(FloatCurve=DataInterface() {propertyValues {Curve {keys [{arriveTangent=1.9165, interpMode=RCIM_Cubic, leaveTangent=1.9165, tangentMode=RCTM_User, time=0, value=0.25}, <at 1 1>]}}})
    )
   )
   &ScaleColor(
    Scale Alpha=&FloatFromCurve(
     FloatCurve=DataInterface() {propertyValues {Curve {
        defaultValue=1
        keys [
         {arriveTangent=1.3196, interpMode=RCIM_Cubic, leaveTangent=1.3196, tangentMode=RCTM_User, time=0, value=0.7}
         {interpMode=RCIM_Cubic, tangentMode=RCTM_User, time=0.5, value=1}
        ]
       }}}
     CurveIndex=$Particles.DistanceScale
    )
    ScaleA=true
    ScaleRGB=true
   )
   /SetVariables(
    Particles.Glow Sprite Camera Offset=5
    Particles.Glow Sprite Color:LinearColor=&MultiplyLinearColor(
     Linear Color=&HueShiftLinearColor(Hue=$System.HueShift, Saturation=1, Value=1, Linear Color=<rgba 1 0.1 0.175 1>)
     Scale Factor=&MakeFloat4FromVectorAndFloat(
      "Vector (RGB)"=<2 2 2>
      "Float (Alpha)"=&FloatFromCurve(
       FloatCurve=DataInterface() {propertyValues {Curve {keys [
           {arriveTangent=8.356, interpMode=RCIM_Cubic, leaveTangent=8.356, tangentMode=RCTM_User, time=0, value=0.1}
           {arriveTangentWeight=0.2855, interpMode=RCIM_Cubic, leaveTangentWeight=0.1894, tangentMode=RCTM_User, time=0.2, value=0.9}
           {arriveTangentWeight=0.6992, interpMode=RCIM_Cubic, tangentMode=RCTM_User, time=1.0011, value=0.002}
          ]}}}
       Scale Curve=&FloatFromCurve(
        FloatCurve=DataInterface() {propertyValues {Curve {
           defaultValue=1
           keys [
            {arriveTangent=1.2232, interpMode=RCIM_Cubic, leaveTangent=1.2232, tangentMode=RCTM_User, time=0, value=0.6}
            {interpMode=RCIM_Cubic, tangentMode=RCTM_User, time=0.5, value=1}
           ]
          }}}
        CurveIndex=$Particles.DistanceScale
       )
      )
     )
    )
    Particles.Glow Sprite Rotation:NiagaraFloat=&Add_Float(
     A=$Particles.Glow Sprite Rotation
     B=&Multiply_Float(
      A=&FloatFromCurve(FloatCurve=DataInterface() {propertyValues {Curve {keys [
           {tangentMode=RCTM_User, time=0, value=1}
           {arriveTangent=-0.3174, interpMode=RCIM_Cubic, leaveTangent=-0.3174, tangentMode=RCTM_User, time=0.3, value=1}
           {arriveTangent=0.0123, interpMode=RCIM_Cubic, leaveTangent=0.0123, tangentMode=RCTM_User, time=1, value=0}
          ]}}})
      B=-5
     )
    )
    Particles.Glow Sprite Size:Vector2f=&Vector2DFromFloat(Value=&Multiply_Float(
      A=&Multiply_Float(A=100, B=$Particles.AmmoTypeScale)
      B=&FloatFromCurve(
       FloatCurve=DataInterface() {propertyValues {Curve {keys [
           {arriveTangent=9.0725, interpMode=RCIM_Cubic, leaveTangent=9.0725, tangentMode=RCTM_User, time=0, value=0}
           {arriveTangent=-2.5249, interpMode=RCIM_Cubic, leaveTangent=-2.5249, tangentMode=RCTM_User, time=0.1408, value=1}
          ]}}}
       Scale Curve=&FloatFromCurve(
        FloatCurve=DataInterface() {propertyValues {Curve {keys [
            {interpMode=RCIM_Cubic, tangentMode=RCTM_Break, time=0, value=0.6}
            {interpMode=RCIM_Cubic, tangentMode=RCTM_Break, time=0.7, value=1}
            <at 1 1>
           ]}}}
        CurveIndex=$Particles.DistanceScale
       )
      )
     ))
   ) {isSetParameters=true}
   &ScaleColor(
    Scale Alpha=&Multiply_Float(
     A=1
     B=/Niagara/DynamicInputs/Camera/CameraFade(
      Camera Query:NiagaraDataInterfaceCamera=DataInterface()
      Fade Offset:NiagaraFloat=250
      Fade Distance:NiagaraFloat=15
      Use Owner Position:NiagaraBool=true
     )
    )
    ScaleA=true
    ScaleRGB=true
   ) {module=ScaleColor001}
  ]
  rendererClasses [&NiagaraMeshRendererProperties, &NiagaraSpriteRendererProperties]
  renderers [
   {rendererClass=&NiagaraMeshRendererProperties, rendererIndex=0}
   {rendererClass=&NiagaraSpriteRendererProperties, rendererIndex=1}
   {rendererClass=&NiagaraSpriteRendererProperties, rendererIndex=2}
  ]
 }
 Decal001 {
  emitterUpdateScript [
   &EmitterState(Scalability Mode=::NewEnumerator1)
   &Spawn_NDCFortImpacts()
   &Spawn_NDCFortImpacts(Physical Material=SurfaceType8, Spawn Count=0, Surface Type=1) {module=Spawn_NDCFortImpacts001}
  ]
  particleSpawnScript [
   /Game/Effects/Niagara/DataChannels/Impacts/InitializeParticleFromNDC_FortImpact(
    AmmoType Scale=1
    FallbackImpactScale=50
    Scale Array=DataInterface() {propertyValues {GpuSyncMode=SyncCpuToGpu, MaxElements=0, FloatData [12,17.5,25,10,40]}}
   )
   /SetVariables(Particles.VisibilityTag=&SelectIntFromArray(Direct Array Index=$Particles.PhysicalSurfaceInteger, Int Selection Array=DataInterface() {propertyValues {IntData [0,1,2,3,0]}})) {isSetParameters=true}
   /SetVariables(Particles.AmmoTypeScale:NiagaraFloat=&SelectFloatFromArray(Direct Array Index=$Particles.AmmoType, Float Selection Array=DataInterface() {propertyValues {FloatData [12,17.5,25,10,40]}})) {enabled=false, isSetParameters=true}
   &InitializeParticle(
    Lifetime=&SelectFloatFromArray(Direct Array Index=$Engine.QualityLevel, Float Selection Array=DataInterface() {propertyValues {FloatData [2,3,4,5]}})
    Mesh Scale=&Multiply_VectorByFloat(
     Vector=<4 1 1>
     Float=&Multiply_Float(A=$Particles.AmmoTypeScale, B=&RandomRangeFloat(Minimum=0.85, Maximum=1))
    )
    Mesh Scale Mode=::NewEnumerator4
    Position=$Particles.Position
    Position Mode=::NewEnumerator1
    Position Offset=<20 0 0>
    UsePositionOffset=false
    Write Material Random=true
   )
   /SetVariables(Particles.DecalEmissiveColor:LinearColor=&SelectLinearColorFromArray(
     Array Sampling Mode=::NewEnumerator1
     Direct Array Index=1
     Color Selection Array=DataInterface() {propertyValues {ColorData [<rgba 0 0 0 1>, <rgba 4 0.5 0.1 1>, <rgba 10 0.6 0.1 1>, <rgba 0.5 3 4 1>, <rgba 1 5 15 1>]}}
     Random Seed=?
    )) {isSetParameters=true}
   /Niagara/Modules/Spawn/Decal/InitialDecalOrientation(
    Orientation Vector:Vector3f=$Particles.ImpactSurfaceNormal
    Rotation:Vector3f=&RandomRangeVector(Minimum=<0 0 0>, Maximum=<0 0 1>)
    Use Rotation Vector=true
   )
  ]
  particleUpdateScript [
   &ParticleState()
   &ScaleColor(
    Scale Alpha=&FloatFromCurve(FloatCurve=DataInterface() {propertyValues {Curve {keys [
         {tangentMode=RCTM_User, time=0, value=1}
         {arriveTangent=-0.0405, interpMode=RCIM_Cubic, leaveTangent=-0.0405, tangentMode=RCTM_User, time=0.7009, value=1}
         {interpMode=RCIM_Cubic, leaveTangent=-8.0089, time=1, value=0}
        ]}}})
    Scale RGB=&Multiply_VectorByFloat(
     Vector=&MakeVectorFromLinearColorRGB(Color=$Particles.DecalEmissiveColor)
     Float=&FloatFromCurve(FloatCurve=DataInterface() {propertyValues {Curve {keys [
          {arriveTangent=-27.9932, interpMode=RCIM_Cubic, leaveTangent=-27.9932, tangentMode=RCTM_User, time=0, value=2}
          {arriveTangent=-0.2721, interpMode=RCIM_Cubic, leaveTangent=-0.2721, tangentMode=RCTM_User, time=0.05, value=0.025}
          <at 0.15 0 RCIM_Cubic>
         ]}}})
    )
    ScaleA=true
    ScaleRGB=true
   )
   &ScaleMeshSize(Scale Factor=<5 5 5>) {enabled=false}
  ]
  rendererClasses [&NiagaraDecalRendererProperties]
  renderers [
   {rendererClass=&NiagaraDecalRendererProperties, rendererIndex=0}
   {rendererClass=&NiagaraDecalRendererProperties, rendererIndex=1}
   {rendererClass=&NiagaraDecalRendererProperties, rendererIndex=2}
   {rendererClass=&NiagaraDecalRendererProperties, rendererIndex=3}
  ]
 }
 NE_FN_PhysMatDebris {
  emitterUpdateScript [
   &EmitterState()
   &Spawn_NDCFortImpacts(Distance Culling=1, LOD Distance=5000, Spawn Count Mode=0)
   &Spawn_NDCFortImpacts(
    Max Spawn Count=5
    Min Spawn Count=3
    Simple Max Spawn Count=2
    Simple Min Spawn Count=1
    Simple Spawn Override=2
    Spawn Count Mode=2
    Surface Type=1
   ) {module=Spawn_NDCFortImpacts001}
   &Spawn_NDCFortImpacts(
    Max Spawn Count=5
    Min Spawn Count=3
    Physical Material=SurfaceType1
    Simple Max Spawn Count=3
    Simple Min Spawn Count=2
    Simple Spawn Override=2
    Spawn Count Mode=2
    Surface Type=1
   ) {module=Spawn_NDCFortImpacts002}
   &Spawn_NDCFortImpacts(
    Max Spawn Count=4
    Min Spawn Count=2
    Physical Material=SurfaceType2
    Simple Max Spawn Count=2
    Simple Min Spawn Count=1
    Simple Spawn Override=2
    Spawn Count Mode=2
    Surface Type=1
   ) {module=Spawn_NDCFortImpacts003}
   &Spawn_NDCFortImpacts(
    Max Spawn Count=7
    Min Spawn Count=3
    Physical Material=SurfaceType18
    Simple Max Spawn Count=2
    Simple Min Spawn Count=1
    Simple Spawn Override=2
    Spawn Count Mode=2
    Surface Type=1
   ) {module=Spawn_NDCFortImpacts004}
   &Spawn_NDCFortImpacts(
    Max Spawn Count=8
    Min Spawn Count=5
    Physical Material=SurfaceType24
    Simple Max Spawn Count=3
    Simple Min Spawn Count=2
    Simple Spawn Override=2
    Spawn Count Mode=2
    Surface Type=1
   ) {module=Spawn_NDCFortImpacts005}
   &Spawn_NDCFortImpacts(
    Max Spawn Count=5
    Min Spawn Count=3
    Physical Material=SurfaceType23
    Simple Max Spawn Count=2
    Simple Min Spawn Count=1
    Simple Spawn Override=2
    Spawn Count Mode=2
    Surface Type=1
   ) {module=Spawn_NDCFortImpacts006}
   &Spawn_NDCFortImpacts(
    Max Spawn Count=5
    Min Spawn Count=3
    Physical Material=SurfaceType10
    Simple Max Spawn Count=2
    Simple Min Spawn Count=1
    Simple Spawn Override=2
    Spawn Count Mode=2
    Surface Type=1
   ) {module=Spawn_NDCFortImpacts007}
   &Spawn_NDCFortImpacts(
    Max Spawn Count=5
    Min Spawn Count=3
    Physical Material=SurfaceType9
    Simple Max Spawn Count=3
    Simple Min Spawn Count=2
    Simple Spawn Override=2
    Spawn Count Mode=2
    Surface Type=1
   ) {module=Spawn_NDCFortImpacts008}
  ]
  particleSpawnScript [
   &InitializeParticleFromNDC_FortImpact(
    AmmoType Scale=1
    Distance Scale=1
    FallbackPosition:NiagaraPosition=<0 0 1>
    Scale Array=DataInterface() {propertyValues {GpuSyncMode=SyncCpuToGpu, MaxElements=4, FloatData [0.8,1,1.2,0.6,1]}}
   )
   /SetVariables(Particles.BiasedRandom=&FloatFromCurve(
     FloatCurve=DataInterface() {propertyValues {Curve {keys [
         <at 0.2 1 RCIM_Cubic>
         {arriveTangent=-3.3394, interpMode=RCIM_Cubic, leaveTangent=-3.3394, tangentMode=RCTM_User, time=0.5, value=0.5}
         <at 0.8 0 RCIM_Cubic>
        ]}}}
     CurveIndex=&RandomRangeFloat(Minimum=0, Maximum=1)
    )) {isSetParameters=true}
   &InitializeParticle(
    Color Mode=::NewEnumerator0
    Lifetime=&RandomRangeFloat(Minimum=0.5, Maximum=1)
    Mass=&Lerp_Float(A=0.6, B=1.2, Alpha=$Particles.BiasedRandom)
    Mass Mode=::NewEnumerator1
    Mesh Scale=&Multiply_VectorByFloat(
     Vector=&RandomRangeVector(Minimum=<0.9 0.4 0.5>, Maximum=<1 1 1>)
     Float=&Lerp_Float(A=0.7, B=1, Alpha=$Particles.BiasedRandom)
    )
    Mesh Scale Mode=::NewEnumerator4
    Position=$Particles.Position
    Position Mode=::NewEnumerator1
    UsePositionOffset=false
    Write Material Random=true
   )
   /SetVariables(Particles.IsSnow:NiagaraBool=&SetBoolByIntComparison(A=$Particles.PhysicalSurfaceInteger, B=23, Comparison Type=::NewEnumerator4)) {isSetParameters=true}
   &FN_PhysicalSurface(
    Dirt Mesh Count=2
    Dirt Scale:NiagaraFloat=&RandomRangeFloat(Minimum=0.33, Maximum=1)
    Dirt Velocity:NiagaraFloat=&RandomRangeFloat(Minimum=150, Maximum=400)
    Glass Mesh Count=3
    Glass Scale=0.5
    Glass Velocity:NiagaraFloat=&RandomRangeFloat(Minimum=100, Maximum=400)
    Glass Velocity Cone Angle:NiagaraFloat=&RandomRangeFloat(Minimum=20, Maximum=70)
    Grass Mass=0.25
    Grass Mesh Count=2
    Grass Scale:NiagaraFloat=&RandomRangeFloat(Minimum=0.66, Maximum=1)
    Grass Velocity:NiagaraFloat=&RandomRangeFloat(Minimum=150, Maximum=400)
    Grass Velocity Cone Angle=30
    Ice Mesh Count=2
    Ice Scale:NiagaraFloat=&MakeCustomFloatFromBool(
     Bool=$Particles.IsSnow
     True Float=&RandomRangeFloat(Minimum=0.8, Maximum=1)
     False Float=&RandomRangeFloat(Minimum=0.33, Maximum=0.5)
    )
    Ice Velocity:NiagaraFloat=&MakeCustomFloatFromBool(
     Bool=$Particles.IsSnow
     True Float=&RandomRangeFloat(Minimum=100, Maximum=300)
     False Float=&RandomRangeFloat(Minimum=100, Maximum=500)
    )
    Ice Velocity Cone Angle:NiagaraFloat=&MakeCustomFloatFromBool(
     Bool=$Particles.IsSnow
     True Float=&RandomRangeFloat(Minimum=50, Maximum=90)
     False Float=&RandomRangeFloat(Minimum=60, Maximum=120)
    )
    Ice Visibility:NiagaraInt32=&MakeCustomIntFromBool(Bool=$Particles.IsSnow, True INT=23, False INT=24)
    Metal Gravity=&RandomRangeFloat(Minimum=-1400, Maximum=-1800)
    Metal Mesh Count=2
    Metal Velocity=&RandomRangeFloat(Minimum=80, Maximum=600)
    Metal Velocity Cone Angle=30
    Physical Surface=/Game/Effects/Niagara/DataChannels/SelectIntBySurface(
     Default:NiagaraInt32=$Particles.PhysicalSurfaceInteger
     Surface:EPhysicalSurface=$Particles.PhysicalSurface
     Num Surfaces:NiagaraInt32=3
     Surface 0:EPhysicalSurface=SurfaceType_Default
     Value For Surface 0:NiagaraInt32=2
     Surface 1:EPhysicalSurface=SurfaceType23
     Value For Surface 1:NiagaraInt32=24
     Surface 2:EPhysicalSurface=SurfaceType9
     Value For Surface 2:NiagaraInt32=&MakeCustomIntFromBool(Bool=&RandomBool(), True INT=9, False INT=10)
     Surface 3:EPhysicalSurface=?
     Value For Surface 3:NiagaraInt32=?
    )
    Stone Mesh Count=3
    Stone Velocity=&RandomRangeFloat(Minimum=150, Maximum=400)
    Stone Velocity Cone Angle:NiagaraFloat=&RandomRangeFloat(Minimum=30, Maximum=120)
    StoneScale:NiagaraFloat=&RandomRangeFloat(Minimum=0.33, Maximum=1)
    Wood Mesh Count=2
    Wood Scale:NiagaraFloat=&RandomRangeFloat(Minimum=0.75, Maximum=1.5)
    Wood Velocity=&RandomRangeFloat(Minimum=80, Maximum=500)
    Wood Velocity Cone Angle=60
   )
   &AddVelocityInCone(
    Cone Angle=$Output.FN_PhysicalSurface.SurfaceResponseConeAngle
    Cone Axis=$Particles.ImpactSurfaceNormal
    Cone Axis Coordinate Space=::Simulation
    Module.Use Velocity Falloff On Cone Axis=false
    Velocity Strength=&Multiply_Float(A=$Output.FN_PhysicalSurface.SurfaceResponseVelocity, B=$Particles.AmmoTypeScale)
   )
   /SetVariables(Particles.Velocity=&Multiply_VectorByFloat(
     Vector=$Particles.Velocity
     Float=&Lerp_Float(A=0.5, B=2, Alpha=&ReturnNormalizedExecIndex())
    )) {isSetParameters=true}
   &AddVelocity(
    Rotation Coordinate Space=::World
    Velocity=<0 0 1>
    Velocity Speed Scale=&Multiply_Float(A=250, B=$Particles.AmmoTypeScale)
   )
   &InitialMeshOrientation(Use Rotation Vector=false)
   &AddRotationalVelocity(Rotation Rate=&RandomVector(Vector Scale=&RandomRangeFloat(Minimum=1, Maximum=2)))
   /SetVariables(Particles.GravityOverrideValue=$Output.FN_PhysicalSurface.GravityOverride) {isSetParameters=true}
   &ApplyInitialForces()
   &AddVelocity() {module=AddVelocity001}
  ]
  particleUpdateScript [
   &ParticleState()
   &Drag(
    Drag=&RandomRangeFloat(Minimum=0.5, Maximum=1)
    Rotational Drag=0.1
    Use Linear Drag=true
    Use Rotational Drag=true
   )
   &GravityForce(Gravity=&Multiply_VectorByFloat(Vector=$Particles.GravityOverrideValue, Float=2))
   &SolveForcesAndVelocity(Clamp Velocity=false, Limit Acceleration=false, Manually Enable Rotational Solver=true)
   &ScaleMeshSize(Scale Factor=&VectorFromFloat(Value=&FloatFromCurve(
      FloatCurve=DataInterface() {propertyValues {Curve {keys [<at 0.6 1 RCIM_Cubic>, {arriveTangent=-6.308, interpMode=RCIM_Cubic, leaveTangent=-6.308, tangentMode=RCTM_User, time=1, value=0}]}}}
      Scale Curve=&Lerp_Float(A=$Particles.AmmoTypeScale, B=$Particles.DistanceScale, Alpha=0.5)
     )))
   &DynamicMaterialParameters(Module.Index0Param0Write=true, Module.Index0Param1Write=true, Module.Index0Param2Write=true, Module.Index0Param3Write=true)
  ]
  rendererClasses [&NiagaraMeshRendererProperties]
  renderers [
   {rendererClass=&NiagaraMeshRendererProperties, rendererIndex=0}
   {rendererClass=&NiagaraMeshRendererProperties, rendererIndex=1}
   {rendererClass=&NiagaraMeshRendererProperties, rendererIndex=2}
   {rendererClass=&NiagaraMeshRendererProperties, rendererIndex=3}
   {rendererClass=&NiagaraMeshRendererProperties, rendererIndex=4}
   {rendererClass=&NiagaraMeshRendererProperties, rendererIndex=5}
   {rendererClass=&NiagaraMeshRendererProperties, rendererIndex=6}
  ]
 }
 Smoke {
  emitterUpdateScript [
   &EmitterState()
   &Spawn_NDCFortImpacts(Distance Culling=1, LOD Distance=6500, Max Spawn Count=9, Min Spawn Count=5, Spawn Count Mode=2)
  ]
  particleSpawnScript [
   &self:Initialize_Particle_from_NDC_PiewPistol_Impacts(
    AmmoType Scale=1
    Distance Scale=1
    LOD Input Distance Max=6500
    LOD Input Distance Min=350
    LOD Output Scale Max=1
    LOD Output Scale Min=0
    Scale Array=DataInterface() {propertyValues {GpuSyncMode=SyncCpuToGpu, MaxElements=4, FloatData [0.8,1,1.25,0.6,1]}}
   ) {module=Initialize_Particle_from_NDC_PiewPistol_Impacts}
   &InitializeParticle(
    Color Minimum=<rgba 1 1 1 0.5>
    Color Mode=::NewEnumerator2
    Lifetime Max=0.65
    Lifetime Min=0.3
    Lifetime Mode=::NewEnumerator1
    Position=$Particles.Position
    Position Mode=::NewEnumerator1
    Sprite Rotation Mode=::NewEnumerator2
    Sprite Size Mode=::NewEnumerator3
    Uniform Sprite Size=&Multiply_Float(
     A=&RandomRangeFloat(Minimum=70, Maximum=100)
     B=&FloatFromCurve(
      FloatCurve=DataInterface() {propertyValues {Curve {keys [
          {interpMode=RCIM_Cubic, tangentMode=RCTM_Break, time=0, value=0}
          {arriveTangent=2.5778, interpMode=RCIM_Cubic, tangentMode=RCTM_Break, time=1, value=1}
         ]}}}
      CurveIndex=&RandomRangeFloat(Minimum=0, Maximum=1)
     )
    )
    UsePositionOffset=false
    Write Material Random=true
   )
   &AddVelocity(
    Cone Angle=&RandomRangeFloat(Minimum=40, Maximum=90)
    Cone Axis=$Particles.ImpactSurfaceNormal
    Module.UseConeFalloff=false
    Module.UseVelDistribution=false
    Override Seed=false
    Velocity Mode=::NewEnumerator2
    Velocity Speed=&RandomRangeFloat(Minimum=0, Maximum=750)
   )
   &CameraOffset(Camera Offset Amount=20)
  ]
  particleUpdateScript [
   &ParticleState()
   &SpriteRotationRate(Rotation Rate=&Multiply_Float(
     A=&RandomRangeFloat(Minimum=-150, Maximum=150)
     B=&FloatFromCurve(FloatCurve=DataInterface() {propertyValues {Curve {keys [{arriveTangent=-2.0511, interpMode=RCIM_Cubic, leaveTangent=-2.0511, tangentMode=RCTM_User, time=0, value=1}, <at 1 0.1 RCIM_Cubic>]}}})
    ))
   &Drag(Drag=12, Use Linear Drag=true, Use Rotational Drag=true)
   &AccelerationForce(Acceleration=&RandomRangeVector(Minimum=<0 0 0>, Maximum=<0 0 750>))
   &SolveForcesAndVelocity(Clamp Velocity=false, Limit Acceleration=false)
   &ScaleSpriteSize(Uniform Curve Sprite Scale=DataInterface() {propertyValues {Curve {keys [
        {arriveTangent=2.2362, interpMode=RCIM_Cubic, leaveTangent=2.2362, tangentMode=RCTM_User, time=0.0013, value=0}
        {arriveTangent=1.953, interpMode=RCIM_Cubic, leaveTangent=1.953, tangentMode=RCTM_User, time=0.5704, value=1.1553}
        {arriveTangent=0.0514, interpMode=RCIM_Cubic, leaveTangent=0.0514, tangentMode=RCTM_User, time=1.0008, value=1.6042}
       ]}}})
   &ScaleColor(
    Scale Alpha=&FloatFromCurve(
     FloatCurve=DataInterface() {propertyValues {Curve {keys [
         {interpMode=RCIM_Cubic, leaveTangent=17.2181, tangentMode=RCTM_Break, time=0, value=0}
         {interpMode=RCIM_Cubic, tangentMode=RCTM_Break, time=0.1, value=1}
         {arriveTangent=-3.5838, interpMode=RCIM_Cubic, tangentMode=RCTM_Break, time=1, value=0}
        ]}}}
     Scale Curve=&FloatFromCurve(
      FloatCurve=DataInterface() {propertyValues {Curve {
         defaultValue=1
         keys [
          {arriveTangent=4.8278, interpMode=RCIM_Cubic, leaveTangent=4.8278, tangentMode=RCTM_User, time=0, value=0}
          {interpMode=RCIM_Cubic, tangentMode=RCTM_User, time=0.7, value=1}
         ]
        }}}
      CurveIndex=$Particles.DistanceScale
      Scale Curve=0.35
     )
    )
    Scale RGB=&MakeVectorFromLinearColorRGB(Color=&HueShiftLinearColor(Hue=$System.HueShift, Saturation=1, Value=1, Linear Color=<rgba 1 0.5 0.5417 1>))
    ScaleA=true
    ScaleRGB=true
   )
   /Game/Effects/Niagara/Modules/Materials/MaterialSmoke(
    Erode:NiagaraFloat=&FloatFromCurve(FloatCurve=DataInterface() {propertyValues {
       bOptimizeLUT=false
       bUseLUT=false
       Curve {keys [
         {arriveTangent=-0.0652, interpMode=RCIM_Cubic, leaveTangent=-0.0652, tangentMode=RCTM_User, time=0.11, value=0.97}
         {arriveTangent=-5.7797, interpMode=RCIM_Cubic, leaveTangent=-5.7797, tangentMode=RCTM_User, time=0.39, value=0.66}
         {arriveTangent=-0.342, interpMode=RCIM_Cubic, leaveTangent=-0.342, tangentMode=RCTM_User, time=0.9954, value=-0.74}
        ]}
      }})
    Shading:NiagaraFloat=&FloatFromCurve(FloatCurve=DataInterface() {propertyValues {
       bOptimizeLUT=false
       bUseLUT=false
       Curve {keys [
         <at 0.07 0.5 RCIM_Cubic>
         {arriveTangent=1.2205, interpMode=RCIM_Cubic, leaveTangent=1.2205, time=0.17, value=1.02}
         {interpMode=RCIM_Cubic, tangentMode=RCTM_User, time=0.3895, value=0.89}
         <at 0.8 0 RCIM_Cubic>
        ]}
      }})
   )
   &DynamicMaterialParameters(
    Index 0 Param 1=&Add_Float(
     A=&RandomRangeFloat(Minimum=0, Maximum=1)
     B=&FloatFromCurve(
      FloatCurve=DataInterface() {propertyValues {Curve {keys [
          {arriveTangent=2.7533, interpMode=RCIM_Cubic, leaveTangent=1.8958, tangentMode=RCTM_User, time=0, value=1}
          {arriveTangent=0.8292, interpMode=RCIM_Cubic, leaveTangent=0.8292, tangentMode=RCTM_User, time=0.321, value=1.3453}
          {arriveTangent=0.0529, interpMode=RCIM_Cubic, leaveTangent=0.2021, tangentMode=RCTM_User, time=1, value=1.5}
         ]}}}
      CurveIndex=&Add_Float(A=&Multiply_Float(A=$Particles.Initial.MaterialRandom, B=0.5), B=$Particles.NormalizedAge)
     )
    )
    Index 0 Param 2=&FloatFromCurve(FloatCurve=DataInterface() {propertyValues {Curve {keys [
         {arriveTangent=0.8564, interpMode=RCIM_Cubic, leaveTangent=0.8564, tangentMode=RCTM_User, time=0.0046, value=-0.9753}
         {arriveTangent=4.2876, interpMode=RCIM_Cubic, leaveTangent=5.5268, tangentMode=RCTM_User, time=1, value=1}
        ]}}})
    Index 0 Param 4=&FloatFromCurve(FloatCurve=DataInterface() {propertyValues {Curve {keys [{arriveTangent=-3.2497, interpMode=RCIM_Cubic, leaveTangent=-3.2497, tangentMode=RCTM_User, time=0, value=1}, <at 1 0 RCIM_Cubic>]}}})
    Index 2 Param 1=&FloatFromCurve(FloatCurve=DataInterface() {propertyValues {Curve {keys [
         {interpMode=RCIM_Cubic, leaveTangent=4.2485, tangentMode=RCTM_Break, time=0, value=0}
         {arriveTangent=1.0548, interpMode=RCIM_Cubic, leaveTangent=1.0548, tangentMode=RCTM_User, time=0.17, value=0.5715}
         <at 1 1>
        ]}}})
    Index 2 Param 2=&RandomRangeFloat(Minimum=0, Maximum=1)
    Index 2 Param 3=&FloatFromCurve(
     FloatCurve=DataInterface() {propertyValues {Curve {keys [
         {interpMode=RCIM_Cubic, leaveTangentWeight=0.3354, tangentMode=RCTM_Break, time=0.4, value=0}
         {arriveTangent=2.7846, interpMode=RCIM_Cubic, leaveTangent=11.1233, tangentMode=RCTM_Break, time=1, value=1}
        ]}}}
     Scale Curve=7
    )
    Index 2 Param 4=&FloatFromCurve(
     FloatCurve=DataInterface() {propertyValues {Curve {keys [
         {interpMode=RCIM_Cubic, leaveTangentWeight=0.4044, tangentMode=RCTM_Break, time=0, value=0.8}
         {arriveTangent=0.4016, arriveTangentWeight=0.382, interpMode=RCIM_Cubic, tangentMode=RCTM_Break, time=1, value=1}
        ]}}}
     Scale Curve=0.5
    )
    Module.Param0WriteEnabled=true
    Module.Param0WriteEnabled2=true
    Module.Param1WriteEnabled=true
    Module.Param1WriteEnabled2=true
    Module.Param2WriteEnabled=true
    Module.Param2WriteEnabled2=true
    Module.Param3WriteEnabled=true
    Module.Param3WriteEnabled2=true
    Write Parameter Index 2=true
   )
  ]
  rendererClasses [&NiagaraSpriteRendererProperties]
  renderers [{rendererClass=&NiagaraSpriteRendererProperties}]
 }
 Sparks_Colored {
  emitterUpdateScript [
   &EmitterState()
   &Spawn_NDCFortImpacts(Distance Culling=1, LOD Distance=8500, Max Spawn Count=15, Min Spawn Count=7, Spawn Count Mode=2)
  ]
  particleSpawnScript [
   &self:Initialize_Particle_from_NDC_PiewPistol_Impacts(
    AmmoType Scale=1
    Distance Scale=1
    LOD Input Distance Max=6500
    LOD Input Distance Min=300
    LOD Output Scale Max=1
    LOD Output Scale Min=0
    Scale Array=DataInterface() {propertyValues {GpuSyncMode=SyncCpuToGpu, MaxElements=4, FloatData [0.8,1,1.25,1,1]}}
   ) {module=Initialize_Particle_from_NDC_PiewPistol_Impacts}
   &InitializeParticle(
    Color=&self:Material_Hue_Shift_01(Color:LinearColor=<rgba 1 0.3636 0.01 1>, Hue Shift:NiagaraFloat=$User.ImpactColorHueControl)
    Lifetime=&MakeCustomFloatFromBool(
     Bool=&SetBoolByIntComparison(A=$Particles.AmmoType, B=3, Comparison Type=::NewEnumerator4)
     True Float=&RandomRangeFloat(Minimum=0.8, Maximum=1.3)
     False Float=&RandomRangeFloat(Minimum=0.2, Maximum=1)
    )
    Position=$Particles.Position
    Position Mode=::NewEnumerator1
    Sprite Size Max=<1 3>
    Sprite Size Min=<0.5 1.5>
    Sprite Size Mode=::NewEnumerator2
    UsePositionOffset=false
    Write Material Random=true
   )
   &ShapeLocation(MaskSpawnGroup=false, Override Seed=false, Sphere Radius=10, UseOwnerScale=true)
   &AddVelocity(
    Cone Angle=35
    Cone Axis=$Particles.ImpactSurfaceNormal
    Module.UseConeFalloff=false
    Module.UseVelDistribution=false
    Override Seed=false
    Velocity Mode=::NewEnumerator2
    Velocity Speed=&Lerp_Float(
     A=300
     B=980
     Alpha=&FloatFromCurve(
      FloatCurve=DataInterface() {propertyValues {Curve {keys [
          {interpMode=RCIM_Cubic, tangentMode=RCTM_User, time=0, value=0}
          {arriveTangent=2.6808, interpMode=RCIM_Cubic, leaveTangent=2.6808, tangentMode=RCTM_User, time=1, value=1}
         ]}}}
      CurveIndex=&RandomRangeFloat(Minimum=0, Maximum=1)
     )
    )
   )
  ]
  particleUpdateScript [
   &ParticleState()
   &GravityForce(Gravity=<0 0 -750>)
   &CurlNoiseForce(
    Bias Noise=false
    Noise Quality / Cost=::NewEnumerator0
    Noise Strength=&FloatFromCurve(
     FloatCurve=DataInterface() {propertyValues {Curve {keys [{interpMode=RCIM_Cubic, tangentMode=RCTM_Break, time=0, value=1}, {interpMode=RCIM_Cubic, tangentMode=RCTM_Break, time=1, value=0}]}}}
     Scale Curve=1000
    )
    Pan Noise=false
   )
   &Drag(
    Drag=&FloatFromCurve(
     FloatCurve=DataInterface() {propertyValues {Curve {keys [
         {interpMode=RCIM_Cubic, leaveTangent=0.4409, leaveTangentWeight=0.5018, tangentMode=RCTM_Break, time=0, value=1}
         {interpMode=RCIM_Cubic, tangentMode=RCTM_Break, time=0.2, value=0}
        ]}}}
     Scale Curve=&MakeCustomFloatFromBool(
      Bool=&SetBoolByIntComparison(A=$Particles.AmmoType, B=3, Comparison Type=::NewEnumerator4)
      True Float=12
      False Float=20
     )
    )
    Use Linear Drag=true
    Use Rotational Drag=false
   )
   &SolveForcesAndVelocity(Clamp Velocity=false, Limit Acceleration=false)
   &ScaleSpriteSize(
    Uniform Curve Scale=&FloatFromCurve(
     FloatCurve=DataInterface() {propertyValues {Curve {
        defaultValue=1
        keys [
         {arriveTangent=0.4785, interpMode=RCIM_Cubic, leaveTangent=0.4785, tangentMode=RCTM_User, time=0, value=0.8}
         {arriveTangent=0.0224, interpMode=RCIM_Cubic, leaveTangent=0.0224, tangentMode=RCTM_User, time=1, value=1}
        ]
       }}}
     CurveIndex=$Particles.DistanceScale
     Scale Curve=2
    )
    Uniform Curve Sprite Scale=DataInterface() {propertyValues {Curve {keys [{interpMode=RCIM_Cubic, tangentMode=RCTM_Break, time=0, value=1}, {interpMode=RCIM_Cubic, tangentMode=RCTM_Break, time=1, value=0.2}]}}}
   )
   &ScaleSpriteSize(
    Scale Sprite Size Mode=::NewEnumerator0
    Uniform Scale Factor=&MakeCustomFloatFromBool(
     Bool=&SetBoolByIntComparison(A=$Particles.AmmoType, B=3, Comparison Type=::NewEnumerator4)
     True Float=&MakeCustomFloatFromBool(
      Bool=&SetBoolByIntComparison(
       A=&ModuloInt(A=&ReturnParticleIDAttributes(), B=$Particles.NDC Spawn Count)
       B=16
       Comparison Type=::NewEnumerator2
      )
      True Float=1
      False Float=0
     )
     False Float=1
    )
   ) {module=ScaleSpriteSize001}
   &ScaleSpriteSizeBySpeed(Max Scale Factor=<1.25 3>)
   &ScaleColor(
    Scale Alpha=&FloatFromCurve(FloatCurve=DataInterface() {propertyValues {Curve {keys [
         {interpMode=RCIM_Cubic, tangentMode=RCTM_User, time=0.0048, value=0.438}
         {arriveTangent=0.3021, interpMode=RCIM_Cubic, leaveTangent=0.3021, tangentMode=RCTM_User, time=0.0983, value=1.5996}
         {arriveTangent=-1.0514, interpMode=RCIM_Cubic, leaveTangent=-1.0514, tangentMode=RCTM_User, time=0.3368, value=0.5383}
         {arriveTangent=-0.8117, interpMode=RCIM_Cubic, leaveTangent=-0.8117, tangentMode=RCTM_User, time=0.5811, value=1.092}
         <at 1 0>
        ]}}})
    Scale RGB=&VectorFromFloat(Value=&FloatFromCurve(FloatCurve=DataInterface() {propertyValues {Curve {keys [{arriveTangent=-52.0046, interpMode=RCIM_Cubic, leaveTangent=-52.0046, tangentMode=RCTM_User, time=0, value=16}, <at 0.8 0.125 RCIM_Cubic>]}}}))
    ScaleA=true
    ScaleRGB=true
   )
   &ScaleColor(
    Scale Alpha=&FloatFromCurve(
     FloatCurve=DataInterface() {propertyValues {Curve {
        defaultValue=1
        keys [
         {arriveTangent=0.5176, interpMode=RCIM_Cubic, leaveTangent=0.5176, tangentMode=RCTM_User, time=0, value=0.1}
         {interpMode=RCIM_Cubic, tangentMode=RCTM_User, time=0.5, value=1}
        ]
       }}}
     CurveIndex=$Particles.DistanceScale
    )
    Scale RGB=&VectorFromFloat(Value=&Lerp_Float(
      A=4
      B=1
      Alpha=&FloatFromCurve(
       FloatCurve=DataInterface() {propertyValues {Curve {keys [
           {interpMode=RCIM_Cubic, tangentMode=RCTM_Break, time=0.4, value=0}
           {arriveTangent=4.1312, interpMode=RCIM_Cubic, tangentMode=RCTM_Break, time=1, value=1}
          ]}}}
       CurveIndex=&RandomRangeFloat(Minimum=0, Maximum=1)
      )
     ))
    ScaleA=true
    ScaleRGB=true
   ) {module=ScaleColor001}
   &DynamicMaterialParameters(
    Index 0 Param 1=2
    Index 0 Param 2=20
    Index 0 Param 3=&SelectFloatFromArray(Direct Array Index=$Particles.VariantIndex, Float Selection Array=DataInterface() {propertyValues {FloatData [0.1,1,1,0.25,1,0.5,1,1]}})
    Module.Index0Float0Write=true
    Module.Index0Float1Write=true
    Module.Index0Float2Write=true
    Module.Index0Float3Write=true
   )
   &ScaleColor(
    Scale Alpha=&Multiply_Float(
     A=1
     B=&CameraFade(
      Camera Query:NiagaraDataInterfaceCamera=DataInterface()
      Fade Offset:NiagaraFloat=250
      Fade Distance:NiagaraFloat=15
      Use Owner Position:NiagaraBool=true
     )
    )
    ScaleA=true
    ScaleRGB=true
   ) {module=ScaleColor002}
  ]
  rendererClasses [&NiagaraSpriteRendererProperties]
  renderers [{rendererClass=&NiagaraSpriteRendererProperties}]
 }
 Streaks {
  emitterUpdateScript [
   &EmitterState()
   &Spawn_NDCFortImpacts(Distance Culling=1, LOD Distance=8500, Max Spawn Count=30, Min Spawn Count=20, Spawn Count Mode=2)
  ]
  particleSpawnScript [
   &self:Initialize_Particle_from_NDC_PiewPistol_Impacts(
    AmmoType Scale=1
    Distance Scale=1
    LOD Input Distance Max=6500
    LOD Input Distance Min=300
    LOD Output Scale Max=1
    LOD Output Scale Min=0
    Scale Array=DataInterface() {propertyValues {GpuSyncMode=SyncCpuToGpu, MaxElements=4, FloatData [0.8,1,1.25,0.6,1]}}
   ) {module=Initialize_Particle_from_NDC_PiewPistol_Impacts}
   &InitializeParticle(
    Color Maximum=&self:Material_Hue_Shift(Color:LinearColor=<rgba 8 2.1534 0.2 1>, Hue Shift:NiagaraFloat=$User.ImpactColorHueControl)
    Color Minimum=&self:Material_Hue_Shift(Color:LinearColor=<rgba 4 1.565 0.1 1>, Hue Shift:NiagaraFloat=$User.ImpactColorHueControl)
    Color Mode=::NewEnumerator2
    Lifetime=&Multiply_Float(
     A=&UniformRangedFloat(Minimum=0.08, Maximum=0.21)
     B=&SelectFloatFromArray(Direct Array Index=$Particles.AmmoType, Float Selection Array=DataInterface() {propertyValues {FloatData [0.6,0.7,1,0.5,0.8]}})
    )
    Mass Mode=::NewEnumerator1
    Position=$Particles.Position
    Position Mode=::NewEnumerator1
    Sprite Size=&RandomRangeVector2D(Minimum=<1 5>, Maximum=<2.5 20>)
    Sprite Size Mode=::NewEnumerator4
    UsePositionOffset=false
    Write Material Random=true
   )
   &AddVelocityInCone(
    Cone Angle=&RandomRangeFloat(Minimum=30, Maximum=60)
    Cone Axis=$Particles.ImpactSurfaceNormal
    Cone Axis Coordinate Space=::World
    Module.Use Velocity Falloff On Cone Axis=false
    Velocity Strength=&Multiply_Float(
     A=&RandomRangeFloat(
      Minimum=/Game/Effects/Niagara/DataChannels/SelectFloatBySurface(
       Default:NiagaraFloat=150
       Surface:EPhysicalSurface=$Particles.PhysicalSurface
       Num Surfaces:NiagaraInt32=1
       Surface 0:EPhysicalSurface=SurfaceType1
       Value For Surface 0:NiagaraFloat=200
       Surface 1:EPhysicalSurface=?
       Value For Surface 1:NiagaraFloat=?
       Surface 2:EPhysicalSurface=?
       Value For Surface 2:NiagaraFloat=?
       Surface 3:EPhysicalSurface=?
       Value For Surface 3:NiagaraFloat=?
      )
      Maximum=&SelectFloatBySurface(
       Default:NiagaraFloat=800
       Surface:EPhysicalSurface=$Particles.PhysicalSurface
       Num Surfaces:NiagaraInt32=1
       Surface 0:EPhysicalSurface=SurfaceType1
       Value For Surface 0:NiagaraFloat=900
       Surface 1:EPhysicalSurface=?
       Value For Surface 1:NiagaraFloat=?
       Surface 2:EPhysicalSurface=?
       Value For Surface 2:NiagaraFloat=?
       Surface 3:EPhysicalSurface=?
       Value For Surface 3:NiagaraFloat=?
      )
     )
     B=$Particles.AmmoTypeScale
    )
   )
  ]
  particleUpdateScript [
   &ParticleState()
   &Drag(Drag=&RandomRangeFloat(Minimum=2, Maximum=2.5), Use Linear Drag=true, Use Rotational Drag=false)
   &SolveForcesAndVelocity(Clamp Velocity=false, Limit Acceleration=false)
   &ScaleSpriteSize(
    Non-Uniform Curve Scale=&Vector2DFromFloat(Value=&RemapRange(Input Value To Remap=$Particles.LODDistance, Input Min=100, Input Max=8000, Output Min=1, Output Max=3))
    Non-Uniform Curve Sprite Scale=DataInterface() {propertyValues {
      XCurve {keys [{tangentMode=RCTM_User, time=0, value=1}, {tangentMode=RCTM_User, time=1, value=0.2}]}
      YCurve {keys [{tangentMode=RCTM_User, time=0, value=1.5}, {tangentMode=RCTM_User, time=1, value=0.7}]}
     }}
    Scale Sprite Size Mode=::NewEnumerator3
   )
   &ScaleColor(
    Scale RGB=&Lerp_Vector(
     A=<0.25 0 0>
     B=<1 1 1>
     Alpha=&FloatFromCurve(FloatCurve=DataInterface() {propertyValues {Curve {keys [<at 0.33 1 RCIM_Cubic>,<at 1 0>]}}})
    )
    ScaleA=true
    ScaleRGB=true
   )
   &ScaleColor(
    Scale Alpha=&Multiply_Float(
     A=1
     B=&CameraFade(
      Camera Query:NiagaraDataInterfaceCamera=DataInterface()
      Fade Offset:NiagaraFloat=250
      Fade Distance:NiagaraFloat=15
      Use Owner Position:NiagaraBool=true
     )
    )
    ScaleA=true
    ScaleRGB=true
   ) {module=ScaleColor001}
  ]
  rendererClasses [&NiagaraSpriteRendererProperties]
  renderers [{rendererClass=&NiagaraSpriteRendererProperties}]
 }
}
summary {
 systemName=NS_NDC_DragonCart_SMG_Handler_Impacts
 emitters [
  {emitterName=NE_FN_PhysMatDebris, rendererClasses [&NiagaraMeshRendererProperties]}
  {emitterName=Base_Flash, rendererClasses [&NiagaraMeshRendererProperties, &NiagaraSpriteRendererProperties]}
  {emitterName=Smoke, rendererClasses [&NiagaraSpriteRendererProperties]}
  {emitterName=Sparks_Colored, rendererClasses [&NiagaraSpriteRendererProperties]}
  {emitterName=Streaks, rendererClasses [&NiagaraSpriteRendererProperties]}
  {emitterName=Decal001, rendererClasses [&NiagaraDecalRendererProperties]}
 ]
 userVariables [User.ImpactColorHueControl {type=NiagaraFloat, defaultValue=0, description=""}]
}
